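# Animation Blueprint for the Innovator Female pickaxe skeleton; its output is only the ref pose through DefaultSlot.
# An unconnected blend chain offsets three Dyn_Blade bones, apparently to extend the blade in stages, using bIsPlayingMeleeAnim, Stage1 and Stage2.
# The update event uses delays to set Sword extend and Stage2 during melee, and otherwise zeroes both stages.
# OnWeaponChanged prints a debug "Hello World"; several other nodes are left unwired.
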AnimBlueprint /BRCosmetics/Weapons/FORT_Melee/Pickaxe_Innovator_Female/Meshes/Innovator_Female_Skeleton_AnimBP
target_skeleton=&/Innovator_Female_Skeleton
graphs [
 AnimGraph {nodes [
   Blend Poses by bool(bActiveValue=true) {blendType=Linear}
   n1 = "Transform (Modify) Bone - Bone: Dyn_Blade3"(Alpha=1, Translation="0,0,0", Rotation="0,0,0", Scale="1,1,1") {translationMode=BMM_Additive, translationSpace=BCS_BoneSpace, alphaBoolBlend {blendOption=Linear}, boneToModify {boneName=Dyn_Blade3}}
   Component To Local(ComponentPose=@n1)
   Output Pose(Result=Slot 'DefaultSlot'(Source=Local Space Ref Pose())) {layerGroup=None}
   n2 = "Transform (Modify) Bone - Bone: Dyn_Blade1"(Alpha=1, Translation="0,0,0", Rotation="0,0,0", Scale="1,1,1") {translationMode=BMM_Additive, translationSpace=BCS_BoneSpace, alphaBoolBlend {blendOption=Linear}, boneToModify {boneName=Dyn_Blade1}}
   n3 = "Transform (Modify) Bone - Bone: Dyn_Blade3"(ComponentPose=Local To Component(LocalPose=Input Pose()), Alpha=1, Translation="11,0,0", Rotation="0,0,0", Scale="1,1,1") {
    translationMode=BMM_Additive
    translationSpace=BCS_BoneSpace
    alphaBoolBlend {blendOption=Linear}
    boneToModify {boneName=Dyn_Blade3}
    translation {x=11}
   }
   n4 = "Transform (Modify) Bone - Bone: Dyn_Blade2"(ComponentPose=@n3, Alpha=1, Translation="10.4,0,0", Rotation="0,0,0", Scale="1,1,1") {
    translationMode=BMM_Additive
    translationSpace=BCS_BoneSpace
    alphaBoolBlend {blendOption=Linear}
    boneToModify {boneName=Dyn_Blade2}
    translation {x=10.4}
   }
   n5 = "Transform (Modify) Bone - Bone: Dyn_Blade1"(ComponentPose=@n4, Alpha=1, Translation="12,0,0", Rotation="0,0,0", Scale="1,1,1") {
    translationMode=BMM_Additive
    translationSpace=BCS_BoneSpace
    alphaBoolBlend {blendOption=Linear}
    boneToModify {boneName=Dyn_Blade1}
    translation {x=12}
   }
   n6 = "Transform (Modify) Bone - Bone: Dyn_Blade2"(Alpha=1, Translation="0,0,0", Rotation="0,0,0", Scale="1,1,1") {translationMode=BMM_Additive, translationSpace=BCS_BoneSpace, alphaBoolBlend {blendOption=Linear}, boneToModify {boneName=Dyn_Blade2}}
   n7 = Two Way Blend!(A=Component To Local(ComponentPose=@n5), B=Component To Local(ComponentPose=@n6), bAlphaBoolEnabled:bool=Get bIsPlayingMeleeAnim()) -> (Pose) {blendNode {
     alphaInputType=Bool
     alphaBoolBlend {blendInTime=0.1, blendOption=ExpInOut, blendOutTime=0.8}
     alphaScaleBiasClamp {bInterpResult=true, interpSpeedIncreasing=5}
    }}
   Two Way Blend!(
    A=Two Way Blend(A=@n7, Alpha=Get Stage1() -> (Stage1:double)) {blendNode {alphaBoolBlend {blendOption=Linear}, alphaScaleBiasClamp {bInterpResult=true, interpSpeedIncreasing=5}}}
    B=Component To Local(ComponentPose=@n2)
    Alpha:float=Get Stage2() -> (Stage2:double)
   ) -> (Pose) {blendNode {alphaBoolBlend {blendOption=Linear}, alphaScaleBiasClamp {bInterpResult=true, interpSpeedIncreasing=5}}}
  ]}
 EventGraph {nodes [
   Event BlueprintUpdateAnimation()
   > n1 = Branch(Condition=Get bIsPlayingMeleeAnim())
   > Set Sword extend(Sword extend:double=1) -> (Output_Get:double)
   > Delay(Duration=0.2)
   > Set Stage2(Stage2:double=1) -> (Output_Get:double)
   > Delay(Duration=0.2)
   > Set Sword extend(Sword extend=1)
   Set Stage1(execute=@n1.else, Stage1:double=0) -> (Output_Get:double)
   > Set Stage2(Stage2=0)
   Event OnWeaponChanged()
   > PrintString!(execute, InString=Hello World, bPrintToScreen, bPrintToLog, TextColor, Duration, Key) -> (then)
   TryGetPawnOwner()
   PickaxeItemDefinitionHasAnyMetaTag(Tags=Make Array("[0]"=(TagName="Cosmetics.MetaTags.BackSpin")), bRequiresEquipped=true)
   CharacterItemDefinitionHasAnyMetaTag()
   Delay(Duration=20)
   Set Stage1(Stage1=1)
  ]}
 NewAnimationLayer_0 {nodes [Output Pose() {layerGroup=None}]}
]
variables [
 Stage1 {type=number}
 Stage2 {type=number}
 Sword extend {type=number}
 Sword Not In use {type=boolean}
]
